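# Control Rig for the GnocchiTea female character's FaceAcc part that applies corrective offsets to hair bones as the head rotates.
# It reads head rotation relative to spine_05 and remaps negative yaw into a 0 to 1 weight for two chained Modify Transforms nodes.
# The first pose applies only when neither isAltSkinEnabled nor is Backbling is set.
# The second applies whenever isAltSkinEnabled is set, regardless of the backbling flag.
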ControlRigBlueprint /BRCosmetics/Characters/Player/Female/Medium/Bodies/F_MED_GnocchiTea/Meshes/Parts/F_MED_GnocchiTea_FaceAcc_CtrlRig
graphs [
 RigVMFunctionLibrary {links [], nodes []}
 RigVMModel {nodes [
   RigUnit_BeginExecution = &RigUnit_BeginExecution()
   RigVMFunction_MathBoolNot = &RigVMFunction_MathBoolNot(Value=RigVMVariableNode() {variable=isAltSkinEnabled}.Value)
   RigVMFunction_MathBoolAnd = &RigVMFunction_MathBoolAnd(
    A=@RigVMFunction_MathBoolNot.Result
    B=&RigVMFunction_MathBoolNot(Value=RigVMVariableNode() {variable=is Backbling}.Value).Result
   )
   Cast = /Script/RigVM.RigVMFunction_MathQuaternionToRotator(Value=&RigUnit_GetRelativeTransformForItem(Child=(Type=Bone,Name="head"), Parent=(Type=Bone,Name="spine_05"), bChildInitial=False, bParentInitial=False).RelativeTransform.Rotation) {is=struct RigVMUnitNode, template="Cast::Execute(in Value,out Result)"}
   Remap = &RigVMFunction_MathDoubleRemap(
    SourceMaximum=-90
    SourceMinimum=-10
    TargetMaximum=1
    TargetMinimum=0
    Value=RigVMDispatchNode(Struct=@Cast.Result) {template="DISPATCH_RigVMDispatch_BreakStruct(in Struct,out Elements)"}.Elements.Yaw
   )
   Multiply = &RigVMFunction_MathDoubleMul(A=&RigVMFunction_MathBoolToFloat(Value=@RigVMFunction_MathBoolAnd.Result).Result, B=@Remap.Result)
   &RigUnit_ModifyTransforms(
    ItemToModify="((Item=(Type=Bone,Name=\"dyn_backHair_corr\"),Transform=(Rotation=(X=0,Y=0,Z=-0.3034,W=0.9529),Translation=(X=-3.2839,Y=-3.2996,Z=0.0227),Scale3D=(X=1,Y=1,Z=1))),(Item=(Type=Bone,Name=\"MSTR_hair\"),Transform=(Rotation=(X=0,Y=0,Z=-0.558,W=0.8298),Translation=( ...(truncated)"
    WeightMinimum=0
    Weight=@Multiply.Result
    ExecutePin=@RigUnit_BeginExecution.ExecutePin
   )
   n1 = RigVMVariableNode() {variable=isAltSkinEnabled}
   n2 = RigVMDispatchNode(Struct=&RigVMFunction_MathQuaternionToRotator(Value=&RigUnit_GetRelativeTransformForItem(Child=(Type=Bone,Name="head"), Parent=(Type=Bone,Name="spine_05"), bChildInitial=False, bParentInitial=False).RelativeTransform.Rotation) {template="Cast::Execute(in Value,out Result)"}.Result) {template="DISPATCH_RigVMDispatch_BreakStruct(in Struct,out Elements)"}
   > &RigUnit_ModifyTransforms(
    ItemToModify="((Item=(Type=Bone,Name=\"dyn_backHair_corr\"),Transform=(Rotation=(X=0,Y=0,Z=-0.3038,W=0.9527),Translation=(X=-2.1151,Y=-2.9797,Z=0),Scale3D=(X=1,Y=1,Z=1))),(Item=(Type=Bone,Name=\"dyn_hair01\"),Transform=(Rotation=(X=0,Y=0,Z=0,W=1),Translation=(X=6.4819,Y=-1. ...(truncated)"
    WeightMinimum=0
    Weight=&RigVMFunction_MathDoubleMul(
     A=&RigVMFunction_MathBoolToFloat(Value=@n1.Value).Result
     B=&RigVMFunction_MathDoubleRemap(SourceMaximum=-90, SourceMinimum=-10, TargetMaximum=1, TargetMinimum=0, Value=@n2.Elements.Yaw).Result
    ).Result
   )
  ]}
]
hierarchy [
 root {loc=<0 0 0>, rot=<0 0 0 1>}
 pelvis {parent=root, loc=<0 -0.5301 98.0571>, rot=<0.0127 0.707 -0.0127 -0.707>}
 spine_01 {parent=pelvis, loc=<6.1208 -0.9328 0>, rot=<0 0 0.034 -0.9994>}
 spine_02 {parent=spine_01, loc=<6.5997 -0.2095 0>, rot=<0 0 0.0238 -0.9997>}
 spine_03 {parent=spine_02, loc=<6.4914 -0.1237 0>, rot=<0 0 0.0149 -0.9999>}
 spine_04 {parent=spine_03, loc=<6.9598 0.0794 0>, rot=<0 0 -0.0132 -0.9999>}
 spine_05 {parent=spine_04, loc=<6.9987 0 0>, rot=<0 0 -0.0015 -1>}
 neck_01 {parent=spine_05, loc=<11.3138 0.1529 0>, rot=<0 0 -0.2003 -0.9797>}
 neck_02 {parent=neck_01, loc=<4.5627 0 0>, rot=<0 0 0.0504 -0.9987>}
 head {parent=neck_02, loc=<5.3914 0 0>, rot=<0 0 0.115 -0.9934>}
 dyn_backHair_corr {parent=head, loc=<1.6664 -5.0072 -0.0725>, rot=<0.2153 -0.9766 0 0>}
 MSTR_hair {parent=head, loc=<-2.2717 -6.8217 -0.0172>, rot=<0.2248 -0.9744 -0.0004 0.0016>}
 dyn_hair01 {parent=MSTR_hair, loc=<0 0 0>, rot=<0 0 0 1>}
 dyn_hair02 {parent=dyn_hair01, loc=<9.3417 0 0>, rot=<0.0011 -0.0027 -0.0583 -0.9983>}
 dyn_hair03 {parent=dyn_hair02, loc=<5.9982 0 0>, rot=<-0.0004 0.0017 -0.0826 -0.9966>}
 dyn_hair04 {parent=dyn_hair03, loc=<6.3403 0 0>, rot=<-0.0008 0.0053 -0.0229 -0.9997>}
 dyn_hair05 {parent=dyn_hair04, loc=<6.0992 -0.0716 0.0547>, rot=<0.0019 -0.0205 -0.0269 -0.9994>}
 dyn_hair06 {parent=dyn_hair05, loc=<7.9421 0 0>, rot=<-0.0011 0.0337 -0.0316 -0.9989>}
 MSTR_curlBack_l {parent=head, loc=<8.5339 7.3311 -5.2574>, rot=<0.163 -0.7063 -0.6888 -0.0105>}
 dyn_curlBack01_l {parent=MSTR_curlBack_l, loc=<0 0 0>, rot=<0 0 0 1>}
 dyn_curlBack02_l {parent=dyn_curlBack01_l, loc=<4.1871 0 0>, rot=<0.0117 0.0724 -0.0848 -0.9937>}
 dyn_curlBack03_l {parent=dyn_curlBack02_l, loc=<3.9121 0 0>, rot=<-0.006 0.0663 -0.0894 -0.9938>}
 MSTR_curlMiddle_l {parent=head, loc=<10.2434 8.8709 -3.0615>, rot=<-0.1668 0.5976 0.7766 0.1094>}
 dyn_curlMiddle01_l {parent=MSTR_curlMiddle_l, loc=<0 0 0>, rot=<0 0 0 1>}
 dyn_curlMiddle02_l {parent=dyn_curlMiddle01_l, loc=<2.4632 0 0>, rot=<-0.0044 0.052 -0.0845 -0.9951>}
 MSTR_curlFront_l {parent=head, loc=<10.3906 8.7359 -1.4711>, rot=<-0.1584 0.4758 0.8514 0.1537>}
 dyn_curlFront01_l {parent=MSTR_curlFront_l, loc=<0 0 0>, rot=<0 0 0 1>}
 dyn_curlFront02_l {parent=dyn_curlFront01_l, loc=<2.4405 0 0>, rot=<-0.0089 0.0554 -0.1577 -0.9859>}
 MSTR_curlMiddle_r {parent=head, loc=<7.1306 6.773 5.6496>, rot=<0.0687 -0.7042 0.7043 0.0575>}
 dyn_curlMiddle01_r {parent=MSTR_curlMiddle_r, loc=<0 0 0>, rot=<0 0 0 1>}
 dyn_curlMiddle02_r {parent=dyn_curlMiddle01_r, loc=<3.4481 0 0>, rot=<-0.0008 0.0075 -0.0535 -0.9985>}
 dyn_curlMiddle03_r {parent=dyn_curlMiddle02_r, loc=<2.5148 0 0>, rot=<0 0 0 1>}
 MSTR_curlBack_r {parent=head, loc=<4.3985 3.8105 6.8241>, rot=<-0.0854 -0.6789 0.7236 0.0906>}
 dyn_curlBack01_r {parent=MSTR_curlBack_r, loc=<0 0 0>, rot=<0 0 0 1>}
 clavicle_l {parent=spine_05, loc=<5.4831 -0.1053 -3.2946>, rot=<0.0465 -0.7289 -0.009 -0.6829>}
 upperarm_l {parent=clavicle_l, loc=<9.8658 -1.2136 -0.4105>, rot=<0.0164 -0.3245 -0.0269 -0.9454>}
 lowerarm_l {parent=upperarm_l, loc=<26.465 -0.8345 0.259>, rot=<0.0036 0.0161 -0.2298 -0.9731>}
 clavicle_r {parent=spine_05, loc=<5.4831 -0.1053 3.2946>, rot=<0.7289 0.0465 -0.6829 0.009>}
 upperarm_r {parent=clavicle_r, loc=<-9.8658 1.2136 0.4105>, rot=<0.0164 -0.3245 -0.0269 -0.9454>}
 lowerarm_r {parent=upperarm_r, loc=<-26.465 0.8345 -0.259>, rot=<0.0036 0.0161 -0.2298 -0.9731>}
 thigh_l {parent=pelvis, loc=<-1.0855 -0.2109 -8.7053>, rot=<0.0009 0.0355 0.0155 -0.9993>}
 thigh_r {parent=pelvis, loc=<-1.0855 -0.2109 8.7053>, rot=<0.0355 -0.0009 0.9993 0.0155>}
]
variables [
 is Backbling {type=bool, default=False}
 isAltSkinEnabled {type=bool, default=False}
]
