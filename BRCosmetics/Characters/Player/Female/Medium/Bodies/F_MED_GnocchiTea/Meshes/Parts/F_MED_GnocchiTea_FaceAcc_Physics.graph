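# Physics asset for the face accessory part of the GnocchiTea female medium character, setting up physics for a hanging hair chain.
# Six dyn_hair capsules link in series from a tiny MSTR_hair root capsule, with limited swing and tight twist at each joint.
# Unconstrained capsules on the head, neck, spine, pelvis, clavicles and arms approximate the upper body, apparently as collision for the hair.
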PhysicsAsset /BRCosmetics/Characters/Player/Female/Medium/Bodies/F_MED_GnocchiTea/Meshes/Parts/F_MED_GnocchiTea_FaceAcc_Physics
bodies {
 MSTR_hair [{shapeType=Capsule, length=0.505, radius=0.505, shapeName=MSTR_hair_capsule, rotation=<rot 0 -90 0>}]
 clavicle_l [{
   shapeType=Capsule
   length=5.7381
   radius=6.9696
   shapeName=clavicle_l_capsule
   center=<3.9914 -0.3758 -1.1978>
   rotation=<rot -0.0088 -89.9994 98.4627>
  }]
 clavicle_r [{
   shapeType=Capsule
   length=5.7381
   radius=6.9696
   shapeName=clavicle_l_capsule
   center=<-3.9914 0.3758 1.1978>
   rotation=<rot -0.0088 -89.9994 -81.5373>
  }]
 dyn_hair01 [{
   shapeType=Capsule
   length=5.5347
   radius=2.2929
   shapeName=dyn_hair01_capsule
   center=<4.4106 0.0585 0.1333>
   rotation=<rot -0.1659 91.7966 -89.006>
  }]
 dyn_hair02 [{
   shapeType=Capsule
   length=3.6882
   radius=2.2198
   shapeName=dyn_hair02_capsule
   center=<2.9184 0.0497 0.1544>
   rotation=<rot -0.0782 85.7449 -88.2954>
  }]
 dyn_hair03 [{
   shapeType=Capsule
   length=2.6714
   radius=2.1495
   shapeName=dyn_hair03_capsule
   center=<2.9213 -0.0688 -0.0459>
   rotation=<rot 0 -94.8 -88.356>
  }]
 dyn_hair04 [{
   shapeType=Capsule
   length=2.1834
   radius=2.2279
   shapeName=dyn_hair04_capsule
   center=<2.1593 -0.1389 0.0233>
   rotation=<rot 0 87.9901 -88.8936>
  }]
 dyn_hair05 [{
   shapeType=Capsule
   length=5.3438
   radius=2.9161
   shapeName=dyn_hair05_capsule
   center=<3.8617 0.0247 -0.1658>
   rotation=<rot -0.4598 89.592 -84.8354>
  }]
 dyn_hair06 [{
   shapeType=Capsule
   length=5.315
   radius=3.4501
   shapeName=dyn_hair06_capsule
   center=<4.5413 -0.3304 -0.0934>
   rotation=<rot -0.7326 -94.6202 -84.8202>
  }]
 head [{shapeType=Capsule, length=25.3865, radius=10.0001, shapeName=head_capsule, center=<5.6006 2.7645 0.0274>}]
 lowerarm_l [{
   shapeType=Capsule
   length=17.779
   radius=4.9184
   shapeName=lowerarm_l_capsule
   center=<9.294 0.0711 0>
   rotation=<rot 0.0161 -92.6454 90.3505>
  }]
 lowerarm_r [{
   shapeType=Capsule
   length=17.779
   radius=4.9184
   shapeName=lowerarm_l_capsule
   center=<-9.294 -0.0711 0>
   rotation=<rot 0.0161 -92.6454 -89.6495>
  }]
 neck_01 [{
   shapeType=Capsule
   length=6.9435
   radius=5.0443
   shapeName=neck_01_capsule
   center=<3.2492 0.5601 -0.0416>
   rotation=<rot 0.1205 -99.4594 90.7234>
  }]
 pelvis [{shapeType=Capsule, length=12.7842, radius=10.365, shapeName=pelvis_capsule, center=<-4.8627 -3.0223 0>, rotation=<rot 0 -90 0>}]
 spine_01 [{shapeType=Capsule, length=10.676, radius=7.8682, shapeName=spine_01_capsule, center=<-0.9742 0.6612 0>, rotation=<rot 0 -90 0>}]
 spine_02 [{shapeType=Capsule, length=5.4693, radius=7.4974, shapeName=spine_02_capsule, center=<0 2.3754 0>, rotation=<rot 0 -90 0>}]
 spine_04 [{shapeType=Capsule, length=4.8017, radius=10.5735, shapeName=spine_04_capsule, center=<0 2.8614 0>, rotation=<rot 0 -90 0>}]
 spine_05 [
  {shapeType=Capsule, length=43.1464, radius=9.0374, shapeName=spine_05_capsule, center=<0.775 -0.202 0>, rotation=<rot 0 -90 0>}
  {shapeType=Capsule, length=42.8905, radius=9.6758, shapeName=spine_05_capsule, center=<10.2702 5.1039 0>}
 ]
 upperarm_l [{
   shapeType=Capsule
   length=20.9286
   radius=5.3889
   shapeName=upperarm_l_capsule
   center=<11.0903 0 -0.3597>
   rotation=<rot -88.8656 -179.8284 -179.3085>
  }]
 upperarm_r [{
   shapeType=Capsule
   length=20.9286
   radius=5.3889
   shapeName=upperarm_l_capsule
   center=<-11.0903 0 0.3597>
   rotation=<rot -88.8656 -179.8284 0.6915>
  }]
}
constraints [
 {bone1Name=dyn_hair02, bone2Name=dyn_hair01, swing1LimitDegrees=25, swing2LimitDegrees=25, twistLimitDegrees=5}
 {bone1Name=dyn_hair01, bone2Name=MSTR_hair, swing1LimitDegrees=30, swing2LimitDegrees=20, twistLimitDegrees=5}
 {bone1Name=dyn_hair03, bone2Name=dyn_hair02, swing1LimitDegrees=20, swing2LimitDegrees=20, twistLimitDegrees=5}
 {bone1Name=dyn_hair04, bone2Name=dyn_hair03, swing1LimitDegrees=20, swing2LimitDegrees=20, twistLimitDegrees=5}
 {bone1Name=dyn_hair05, bone2Name=dyn_hair04, swing1LimitDegrees=20, swing2LimitDegrees=20, twistLimitDegrees=5}
 {bone1Name=dyn_hair06, bone2Name=dyn_hair05, swing1LimitDegrees=45, swing2LimitDegrees=20, twistLimitDegrees=2}
]
