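# Post-process-style AnimBP for the GnocchiTea female face accessory part, driving hair secondary motion on the copied parent pose.
# A Control Rig, weighted bone constraints and offsets, including head-up correctives when wearing backbling, shape the pose.
# Two RigidBody sims feed a cached pose, then serial Anim Dynamics nodes swing the left and right curls.
# The EventGraph detects backbling, picks matching skydive wind/gravity settings and updates gravity and corrective alphas.
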AnimBlueprint /BRCosmetics/Characters/Player/Female/Medium/Bodies/F_MED_GnocchiTea/Meshes/Parts/F_MED_GnocchiTea_FaceAcc_AnimBP
target_skeleton=&/F_MED_GnocchiTea_FaceAcc_Skeleton
graphs [
 AnimGraph {nodes [
   n1 = Anim Dynamics - Bone: dyn_curlBack01_l!(ComponentPose=Local To Component(LocalPose=Use cached pose 'RBANSim'()), Alpha=1, GravityOverride:Vector="0,0,-980", ExternalForce="0,0,0") -> (Pose) {
    angularDampingOverride=0.999
    angularSpringConstant=100
    bAngularSpring=true
    bOverrideAngularDamping=true
    bUseGravityOverride=true
    simulationSpace=RootRelative
    alphaBoolBlend {blendOption=Linear}
    boundBone {boneName=dyn_curlBack01_l}
    componentAppliedLinearAccClamp=<10000 10000 10000>
    componentLinearAccScale=<0.1 0.1 0.35>
    componentLinearVelScale=<0.1 0.1 0.35>
    physicsBodyDefinitions [{
      boundBone {boneName=dyn_curlBack01_l}
      boxExtents=<25 25 25>
      constraintSetup {angularLimitsMax=<0 5 0.5>, angularLimitsMin=<0 -1 -10>, angularTarget=<1 0 0>}
      localJointOffset=<0.1 0 0>
     }]
   }
   n2 = Anim Dynamics - Bone: dyn_curlBack02_l!(ComponentPose=@n1, Alpha=1, GravityOverride:Vector="0,0,-980", ExternalForce="0,0,0") -> (Pose) {
    angularDampingOverride=0.999
    angularSpringConstant=100
    bAngularSpring=true
    bOverrideAngularDamping=true
    bUseGravityOverride=true
    simulationSpace=RootRelative
    alphaBoolBlend {blendOption=Linear}
    boundBone {boneName=dyn_curlBack02_l}
    componentAppliedLinearAccClamp=<10000 10000 10000>
    componentLinearAccScale=<0.1 0.1 0.35>
    componentLinearVelScale=<0.1 0.1 0.35>
    physicsBodyDefinitions [{
      boundBone {boneName=dyn_curlBack02_l}
      boxExtents=<20 20 20>
      constraintSetup {angularLimitsMax=<0 5 3>, angularLimitsMin=<0 -3 -10>, angularTarget=<1 0 0>}
      localJointOffset=<0.1 0 0>
     }]
    relativeSpaceBone {boneName=dyn_curlBack01_l}
   }
   n3 = Anim Dynamics - Bone: dyn_curlBack03_l!(ComponentPose=@n2, Alpha=1, GravityOverride:Vector="0,0,-980", ExternalForce="0,0,0") -> (Pose) {
    angularDampingOverride=0.999
    angularSpringConstant=100
    bAngularSpring=true
    bOverrideAngularDamping=true
    bUseGravityOverride=true
    simulationSpace=RootRelative
    alphaBoolBlend {blendOption=Linear}
    boundBone {boneName=dyn_curlBack03_l}
    componentAppliedLinearAccClamp=<10000 10000 10000>
    componentLinearAccScale=<0.1 0.1 0.35>
    componentLinearVelScale=<0.1 0.1 0.35>
    physicsBodyDefinitions [{
      boundBone {boneName=dyn_curlBack03_l}
      boxExtents=<20 20 20>
      constraintSetup {angularLimitsMax=<0 5 10>, angularLimitsMin=<0 -5 -10>, angularTarget=<1 0 0>}
      localJointOffset=<0.1 0 0>
     }]
    relativeSpaceBone {boneName=dyn_curlBack02_l}
   }
   n4 = Anim Dynamics - Bone: dyn_curlMiddle01_l!(ComponentPose=@n3, Alpha=1, GravityOverride:Vector="0,0,-980", ExternalForce="0,0,0") -> (Pose) {
    angularDampingOverride=0.999
    angularSpringConstant=100
    bAngularSpring=true
    bOverrideAngularDamping=true
    bUseGravityOverride=true
    simulationSpace=RootRelative
    alphaBoolBlend {blendOption=Linear}
    boundBone {boneName=dyn_curlMiddle01_l}
    componentAppliedLinearAccClamp=<10000 10000 10000>
    componentLinearAccScale=<0.1 0.1 0.35>
    componentLinearVelScale=<0.1 0.1 0.35>
    physicsBodyDefinitions [{
      boundBone {boneName=dyn_curlMiddle01_l}
      boxExtents=<25 25 25>
      constraintSetup {angularLimitsMax=<0 5 0.5>, angularLimitsMin=<0 -1 -10>, angularTarget=<1 0 0>}
      localJointOffset=<0.1 0 0>
     }]
   }
   n5 = Anim Dynamics - Bone: dyn_curlMiddle02_l!(ComponentPose=@n4, Alpha=1, GravityOverride:Vector="0,0,-980", ExternalForce="0,0,0") -> (Pose) {
    angularDampingOverride=0.999
    angularSpringConstant=100
    bAngularSpring=true
    bOverrideAngularDamping=true
    bUseGravityOverride=true
    simulationSpace=RootRelative
    alphaBoolBlend {blendOption=Linear}
    boundBone {boneName=dyn_curlMiddle02_l}
    componentAppliedLinearAccClamp=<10000 10000 10000>
    componentLinearAccScale=<0.1 0.1 0.35>
    componentLinearVelScale=<0.1 0.1 0.35>
    physicsBodyDefinitions [{
      boundBone {boneName=dyn_curlMiddle02_l}
      boxExtents=<20 20 20>
      constraintSetup {angularLimitsMax=<0 5 3>, angularLimitsMin=<0 -3 -10>, angularTarget=<1 0 0>}
      localJointOffset=<0.1 0 0>
     }]
    relativeSpaceBone {boneName=dyn_curlMiddle01_l}
   }
   n6 = Anim Dynamics - Bone: dyn_curlFront01_l!(ComponentPose=@n5, Alpha=1, GravityOverride:Vector="0,0,-980", ExternalForce="0,0,0") -> (Pose) {
    angularDampingOverride=0.999
    angularSpringConstant=100
    bAngularSpring=true
    bOverrideAngularDamping=true
    bUseGravityOverride=true
    simulationSpace=RootRelative
    alphaBoolBlend {blendOption=Linear}
    boundBone {boneName=dyn_curlFront01_l}
    componentAppliedLinearAccClamp=<10000 10000 10000>
    componentLinearAccScale=<0.1 0.1 0.35>
    componentLinearVelScale=<0.1 0.1 0.35>
    physicsBodyDefinitions [{
      boundBone {boneName=dyn_curlFront01_l}
      boxExtents=<25 25 25>
      constraintSetup {angularLimitsMax=<0 5 0.5>, angularLimitsMin=<0 -1 -10>, angularTarget=<1 0 0>}
      localJointOffset=<0.1 0 0>
     }]
   }
   n7 = Anim Dynamics - Bone: dyn_curlFront02_l!(ComponentPose=@n6, Alpha=1, GravityOverride:Vector="0,0,-980", ExternalForce="0,0,0") -> (Pose) {
    angularDampingOverride=0.999
    angularSpringConstant=100
    bAngularSpring=true
    bOverrideAngularDamping=true
    bUseGravityOverride=true
    simulationSpace=RootRelative
    alphaBoolBlend {blendOption=Linear}
    boundBone {boneName=dyn_curlFront02_l}
    componentAppliedLinearAccClamp=<10000 10000 10000>
    componentLinearAccScale=<0.1 0.1 0.35>
    componentLinearVelScale=<0.1 0.1 0.35>
    physicsBodyDefinitions [{
      boundBone {boneName=dyn_curlFront02_l}
      boxExtents=<20 20 20>
      constraintSetup {angularLimitsMax=<0 5 3>, angularLimitsMin=<0 -3 -10>, angularTarget=<1 0 0>}
      localJointOffset=<0.1 0 0>
     }]
    relativeSpaceBone {boneName=dyn_curlFront01_l}
   }
   n8 = Anim Dynamics - Bone: dyn_curlMiddle01_r!(ComponentPose=@n7, Alpha=1, GravityOverride:Vector="0,0,-980", ExternalForce="0,0,0") -> (Pose) {
    angularDampingOverride=0.999
    angularSpringConstant=100
    bAngularSpring=true
    bOverrideAngularDamping=true
    bUseGravityOverride=true
    simulationSpace=RootRelative
    alphaBoolBlend {blendOption=Linear}
    boundBone {boneName=dyn_curlMiddle01_r}
    componentAppliedLinearAccClamp=<10000 10000 10000>
    componentLinearAccScale=<0.1 0.1 0.35>
    componentLinearVelScale=<0.1 0.1 0.35>
    physicsBodyDefinitions [{
      boundBone {boneName=dyn_curlMiddle01_r}
      boxExtents=<25 25 25>
      constraintSetup {angularLimitsMax=<0 5 0.5>, angularLimitsMin=<0 -1 -10>, angularTarget=<1 0 0>}
      localJointOffset=<0.1 0 0>
     }]
   }
   n9 = Anim Dynamics - Bone: dyn_curlMiddle02_r!(ComponentPose=@n8, Alpha=1, GravityOverride:Vector="0,0,-980", ExternalForce="0,0,0") -> (Pose) {
    angularDampingOverride=0.999
    angularSpringConstant=100
    bAngularSpring=true
    bOverrideAngularDamping=true
    bUseGravityOverride=true
    simulationSpace=RootRelative
    alphaBoolBlend {blendOption=Linear}
    boundBone {boneName=dyn_curlMiddle02_r}
    componentAppliedLinearAccClamp=<10000 10000 10000>
    componentLinearAccScale=<0.1 0.1 0.35>
    componentLinearVelScale=<0.1 0.1 0.35>
    physicsBodyDefinitions [{
      boundBone {boneName=dyn_curlMiddle02_r}
      boxExtents=<20 20 20>
      constraintSetup {angularLimitsMax=<0 5 3>, angularLimitsMin=<0 -3 -10>, angularTarget=<1 0 0>}
      localJointOffset=<0.1 0 0>
     }]
    relativeSpaceBone {boneName=dyn_curlMiddle01_r}
   }
   n10 = Anim Dynamics - Bone: dyn_curlMiddle03_r!(ComponentPose=@n9, Alpha=1, GravityOverride:Vector="0,0,-980", ExternalForce="0,0,0") -> (Pose) {
    angularDampingOverride=0.999
    angularSpringConstant=100
    bAngularSpring=true
    bOverrideAngularDamping=true
    bUseGravityOverride=true
    simulationSpace=RootRelative
    alphaBoolBlend {blendOption=Linear}
    boundBone {boneName=dyn_curlMiddle03_r}
    componentAppliedLinearAccClamp=<10000 10000 10000>
    componentLinearAccScale=<0.1 0.1 0.35>
    componentLinearVelScale=<0.1 0.1 0.35>
    physicsBodyDefinitions [{
      boundBone {boneName=dyn_curlMiddle03_r}
      boxExtents=<20 20 20>
      constraintSetup {angularLimitsMax=<0 5 3>, angularLimitsMin=<0 -3 -10>, angularTarget=<1 0 0>}
      localJointOffset=<0.1 0 0>
     }]
    relativeSpaceBone {boneName=dyn_curlMiddle02_r}
   }
   n11 = Anim Dynamics - Bone: dyn_curlFront01_r!(ComponentPose=@n10, Alpha=1, GravityOverride:Vector="0,0,-980", ExternalForce="0,0,0") -> (Pose) {
    angularDampingOverride=0.999
    angularSpringConstant=100
    bAngularSpring=true
    bOverrideAngularDamping=true
    bUseGravityOverride=true
    simulationSpace=RootRelative
    alphaBoolBlend {blendOption=Linear}
    boundBone {boneName=dyn_curlFront01_r}
    componentAppliedLinearAccClamp=<10000 10000 10000>
    componentLinearAccScale=<0.1 0.1 0.35>
    componentLinearVelScale=<0.1 0.1 0.35>
    physicsBodyDefinitions [{
      boundBone {boneName=dyn_curlFront01_r}
      boxExtents=<25 25 25>
      constraintSetup {angularLimitsMax=<0 5 0.5>, angularLimitsMin=<0 -1 -10>, angularTarget=<1 0 0>}
      localJointOffset=<0.1 0 0>
     }]
   }
   n12 = Anim Dynamics - Bone: dyn_curlFront02_r!(ComponentPose=@n11, Alpha=1, GravityOverride:Vector="0,0,-980", ExternalForce="0,0,0") -> (Pose) {
    angularDampingOverride=0.999
    angularSpringConstant=100
    bAngularSpring=true
    bOverrideAngularDamping=true
    bUseGravityOverride=true
    simulationSpace=RootRelative
    alphaBoolBlend {blendOption=Linear}
    boundBone {boneName=dyn_curlFront02_r}
    componentAppliedLinearAccClamp=<10000 10000 10000>
    componentLinearAccScale=<0.1 0.1 0.35>
    componentLinearVelScale=<0.1 0.1 0.35>
    physicsBodyDefinitions [{
      boundBone {boneName=dyn_curlFront02_r}
      boxExtents=<20 20 20>
      constraintSetup {angularLimitsMax=<0 5 3>, angularLimitsMin=<0 -3 -10>, angularTarget=<1 0 0>}
      localJointOffset=<0.1 0 0>
     }]
    relativeSpaceBone {boneName=dyn_curlFront01_r}
   }
   n13 = Anim Dynamics - Bone: dyn_curlFront03_r!(ComponentPose=@n12, Alpha=1, GravityOverride:Vector="0,0,-980", ExternalForce="0,0,0") -> (Pose) {
    angularDampingOverride=0.999
    angularSpringConstant=100
    bAngularSpring=true
    bOverrideAngularDamping=true
    bUseGravityOverride=true
    simulationSpace=RootRelative
    alphaBoolBlend {blendOption=Linear}
    boundBone {boneName=dyn_curlFront03_r}
    componentAppliedLinearAccClamp=<10000 10000 10000>
    componentLinearAccScale=<0.1 0.1 0.35>
    componentLinearVelScale=<0.1 0.1 0.35>
    physicsBodyDefinitions [{
      boundBone {boneName=dyn_curlFront03_r}
      boxExtents=<20 20 20>
      constraintSetup {angularLimitsMax=<0 5 3>, angularLimitsMin=<0 -3 -10>, angularTarget=<1 0 0>}
      localJointOffset=<0.1 0 0>
     }]
    relativeSpaceBone {boneName=dyn_curlFront02_r}
   }
   n14 = Anim Dynamics - Bone: dyn_curlBack01_r!(ComponentPose=@n13, Alpha=1, GravityOverride:Vector="0,0,-980", ExternalForce="0,0,0") -> (Pose) {
    angularDampingOverride=0.999
    angularSpringConstant=100
    bLinearSpring=true
    bOverrideAngularDamping=true
    bOverrideLinearDamping=true
    bUseGravityOverride=true
    linearDampingOverride=0.99
    linearSpringConstant=50
    simulationSpace=RootRelative
    alphaBoolBlend {blendOption=Linear}
    boundBone {boneName=dyn_curlBack01_r}
    componentAppliedLinearAccClamp=<10000 10000 10000>
    componentLinearAccScale=<0.1 0.1 0.35>
    componentLinearVelScale=<0.1 0.1 0.35>
    physicsBodyDefinitions [{
      boundBone {boneName=dyn_curlBack01_r}
      boxExtents=<10 10 10>
      constraintSetup {angularLimitsMax=<3 0 5>, angularLimitsMin=<-3 0 -5>, angularTarget=<1 0 0>, linearAxesMax=<0.4 0 0>, linearAxesMin=<-1 0 0>}
      localJointOffset=<0.5 0 0>
     }]
   }
   Output Pose(Result=Component To Local(ComponentPose=@n14)) {layerGroup=None}
   n15 = Control Rig!(
    Source=Copy Pose From Mesh() {bUseAttachedParent=true}
    Alpha
    isAltSkinEnabled:bool=Get AWS() -> (AWS:bool)
    is Backbling:bool=Get IsWearingBackbling() -> (IsWearingBackbling:bool)
   ) -> (Pose) {
    alphaBoolBlend {blendOption=Linear}
    controlRigAssetReference {blueprintRigClass=&/F_MED_GnocchiTea_FaceAcc_CtrlRig.F_MED_GnocchiTea_FaceAcc_CtrlRig_C}
   }
   n16 = "Constraint - MSTR_hair (PPP)"!(
    ComponentPose=Local To Component(LocalPose=@n15)
    bAlphaBoolEnabled:bool=True
    ConstraintWeights_0=0.1
    ConstraintWeights_1:float=0.6
    ConstraintWeights_2:float=0.3
   ) -> (Pose) {
    alphaInputType=Bool
    alphaBoolBlend {blendOption=Linear}
    boneToModify {boneName=MSTR_hair}
    constraintSetup [{targetBone {boneName=head}}, {targetBone {boneName=spine_05}}, {targetBone {boneName=neck_02}}]
    constraintWeights [1,1,1]
   }
   n17 = "Constraint - dyn_hair02 (PPP)"!(ComponentPose=@n16, bAlphaBoolEnabled:bool=True, ConstraintWeights_0=0.3, ConstraintWeights_1:float=0.3, ConstraintWeights_2:float=0.4) -> (Pose) {
    alphaInputType=Bool
    alphaBoolBlend {blendOption=Linear}
    boneToModify {boneName=dyn_hair02}
    constraintSetup [{targetBone {boneName=spine_05}}, {targetBone {boneName=spine_04}}, {targetBone {boneName=dyn_hair01}}]
    constraintWeights [1,1,1]
   }
   n18 = "Constraint - dyn_hair03 (PP)"!(ComponentPose=@n17, bAlphaBoolEnabled:bool=True, ConstraintWeights_0=0.1, ConstraintWeights_1:float=0.9) -> (Pose) {
    alphaInputType=Bool
    alphaBoolBlend {blendOption=Linear}
    boneToModify {boneName=dyn_hair03}
    constraintSetup [{targetBone {boneName=spine_05}}, {targetBone {boneName=spine_04}}]
    constraintWeights [1,1]
   }
   n19 = "Constraint - dyn_hair04 (PP)"!(ComponentPose=@n18, bAlphaBoolEnabled:bool=True, ConstraintWeights_0=0.8, ConstraintWeights_1:float=0.2) -> (Pose) {
    alphaInputType=Bool
    alphaBoolBlend {blendOption=Linear}
    boneToModify {boneName=dyn_hair04}
    constraintSetup [{targetBone {boneName=spine_04}}, {targetBone {boneName=spine_03}}]
    constraintWeights [1,1]
   }
   n20 = "Constraint - dyn_hair05 (PP)"!(ComponentPose=@n19, bAlphaBoolEnabled:bool=True, ConstraintWeights_0=0.8, ConstraintWeights_1:float=0.2) -> (Pose) {
    alphaInputType=Bool
    alphaBoolBlend {blendOption=Linear}
    boneToModify {boneName=dyn_hair05}
    constraintSetup [{targetBone {boneName=spine_04}}, {targetBone {boneName=spine_03}}]
    constraintWeights [1,1]
   }
   n21 = "Transform (Modify) Bone - Bone: dyn_hair04"(ComponentPose=@n20, Alpha=1, Translation="0,-1,0.5", Rotation="0,0,30", Scale="1,1,1") {
    rotationMode=BMM_Additive
    translationMode=BMM_Additive
    alphaBoolBlend {blendOption=Linear}
    boneToModify {boneName=dyn_hair04}
    rotation {roll=30}
    translation {y=-1, z=0.5}
   }
   n22 = "Transform (Modify) Bone - Bone: dyn_hair05"(ComponentPose=@n21, Alpha=1, Translation="0,-1,0.5", Rotation="0,0,30", Scale="1,1,1") {
    rotationMode=BMM_Additive
    translationMode=BMM_Additive
    alphaBoolBlend {blendOption=Linear}
    boneToModify {boneName=dyn_hair05}
    rotation {roll=30}
    translation {y=-1, z=0.5}
   }
   n23 = "Transform (Modify) Bone - Bone: dyn_hair06"(ComponentPose=@n22, Alpha=1, Translation="0,0,0", Rotation="0,0,35", Scale="1,1,1") {rotationMode=BMM_Additive, alphaBoolBlend {blendOption=Linear}, boneToModify {boneName=dyn_hair06}, rotation {roll=35}}
   n24 = "Constraint - dyn_hair01 (PP)"!(ComponentPose=@n23, bAlphaBoolEnabled:bool=True, ConstraintWeights_0=0.4, ConstraintWeights_1:float=0.6) -> (Pose) {
    alphaInputType=Bool
    alphaBoolBlend {blendOption=Linear}
    boneToModify {boneName=dyn_hair01}
    constraintSetup [{targetBone {boneName=head}}, {targetBone {boneName=spine_05}}]
    constraintWeights [1,1]
   }
   n25 = Get HeadUpAlpha() -> (HeadUpAlpha:double)
   n26 = "Transform (Modify) Bone - Bone: dyn_backHair_corr"(ComponentPose=@n24, Alpha=@n25, Translation="0,1,0", Rotation="0,15,0", Scale="1,1,1") {
    rotationMode=BMM_Additive
    rotationSpace=BCS_BoneSpace
    translationMode=BMM_Additive
    translationSpace=BCS_BoneSpace
    alphaBoolBlend {blendOption=Linear}
    boneToModify {boneName=dyn_backHair_corr}
    rotation {yaw=15}
    translation {y=1}
   }
   n27 = "Transform (Modify) Bone - Bone: dyn_hair01"(ComponentPose=@n26, Alpha=@n25, Translation="2,-2,0", Rotation="0,-15,0", Scale="1,1,1") {
    rotationMode=BMM_Additive
    rotationSpace=BCS_BoneSpace
    translationMode=BMM_Additive
    translationSpace=BCS_BoneSpace
    alphaBoolBlend {blendOption=Linear}
    boneToModify {boneName=dyn_hair01}
    rotation {yaw=-15}
    translation=<2 -2>
   }
   n28 = "Constraint - dyn_hair02 (PP)"!(ComponentPose=@n27, bAlphaBoolEnabled:bool=True, ConstraintWeights_0=0.1, ConstraintWeights_1:float=0.9) -> (Pose) {
    alphaInputType=Bool
    alphaBoolBlend {blendOption=Linear}
    boneToModify {boneName=dyn_hair02}
    constraintSetup [{targetBone {boneName=dyn_hair01}}, {targetBone {boneName=spine_05}}]
    constraintWeights [1,1]
   }
   n29 = "Constraint - dyn_hair03 (PPP)"!(ComponentPose=@n28, bAlphaBoolEnabled:bool=True, ConstraintWeights_0=0.8, ConstraintWeights_1:float=0.2, ConstraintWeights_2:float=0) -> (Pose) {
    alphaInputType=Bool
    alphaBoolBlend {blendOption=Linear}
    boneToModify {boneName=dyn_hair03}
    constraintSetup [{targetBone {boneName=dyn_hair02}}, {targetBone {boneName=spine_04}}, {targetBone {boneName=spine_05}}]
    constraintWeights [1,1,1]
   }
   n30 = RigidBody!(ComponentPose=@n29, Alpha, OverrideWorldGravity:Vector="0,0,0", ExternalForce) -> (Pose) {
    bAllowPoseScalePropagation=false
    bClampLinearTranslationLimitToRefPose=true
    bOverrideWorldGravity=true
    overridePhysicsAsset=&/F_MED_GnocchiTea_FaceAcc_Physics
    simulationSpace=BaseBoneSpace
    alphaBoolBlend {blendOption=Linear}
    baseBoneRef {boneName=spine_05}
    componentAppliedLinearAccClamp=<8000 8000 8000>
    componentLinearAccScale=<0.3 0.2 0.3>
    componentLinearVelScale=<0.2 0.2 0.3>
    simSpaceSettings {maxAngularAcceleration=150, maxAngularVelocity=3, maxLinearAcceleration=75, maxLinearVelocity=50, velocityScaleZ=0, worldAlpha=1}
   }
   n31 = RigidBody!(ComponentPose=@n30, bAlphaBoolEnabled:bool, OverrideWorldGravity="0,0,0", ExternalForce) -> (Pose) {
    alphaInputType=Bool
    bClampLinearTranslationLimitToRefPose=true
    bOverrideWorldGravity=true
    overridePhysicsAsset=&/F_MED_GnocchiTea_FaceAccBackPack_Physics
    simulationSpace=BaseBoneSpace
    alphaBoolBlend {blendOption=Linear}
    baseBoneRef {boneName=spine_05}
    componentLinearAccScale=<0.2 0.1 0.1>
    componentLinearVelScale=<0.1 0.1 0.1>
    simSpaceSettings {maxAngularAcceleration=800, maxAngularVelocity=3, maxLinearAcceleration=400, maxLinearVelocity=250, velocityScaleZ=0, worldAlpha=1}
   }
   Save cached pose 'RBANSim'(Pose=Component To Local(ComponentPose=@n31)) {cacheName=RBANSim}
  ]}
 EventGraph {nodes [
   Event BlueprintUpdateAnimation()
   > n1 = Gravity Override State Function DEPRECATED(CharacterPartAnimInstance=Self-Reference(), GameTimeInSeconds=GetGameTimeInSeconds(), GravityOverrideParams=Get Gravity Override Params())
   > Set Gravity Override(Gravity Override=@n1)
   > Set AWS(AWS:bool=NOT Boolean(A=Get bRigidBodyEnabled())) -> (Output_Get:bool)
   n2 = Get AWS()
   n3 = GetPartSkeletalMeshComponent()
   > Set PonyTailCorrAlpha(PonyTailCorrAlpha:float=float * float(
     A=GetAlphaFromReferenceTransformDelta(SkeletalMeshComponent=@n3, BoneName=spine_03, BaseName=pelvis, InputRange=(X=-10,Y=-60), TwistAxis="1,0,0")
     B="To Float (Boolean)"(InBool=@n2)
    ) {is=title K2Node_PromotableOperator}) -> (Output_Get:float)
   n4 = GetPartSkeletalMeshComponent()
   n5 = Get IsWearingBackbling()
   > Set HeadUpAlpha(HeadUpAlpha:double=float * float(
     A=GetAlphaFromReferenceTransformDelta(SkeletalMeshComponent=@n4, BoneName=head, BaseName=spine_05, InputRange=(X=0,Y=-60), TwistAxis="1,0,0")
     B="To Float (Boolean)"(InBool=@n5)
    )) -> (Output_Get:double)
   n6 = Event BlueprintInitializeAnimation()
   n7 = Sequence(execute=[@n6.then, @n15.then])
   n8 = Branch(execute=@n7.then_1, Condition=Get IsWearingBackbling())
   > n9 = Set members in Gravity Override Params Struct!(
    execute:Exec
    StructRef:"Gravity Override Params Struct Structure (by ref)"=Get Gravity Override Params()
    Skydive_GoingUp_28_#1:Gravity Override Setting Struct Structure=Make Gravity Override Setting Struct!(
     WindFrequency_28_#2:Vector="1.5,5,6"
     WindAmplitude_31_#3:Vector="0,70,0"
     WindOffset_32_#4:Vector="0,0,0"
     GravityOverride_15_#5:Vector="0,-300,-100"
     JointName_27_#6:Name=root
    ) -> (GravityOverrideSettingStruct:Gravity Override Setting Struct Structure) {is=title K2Node_MakeStruct}
    Skydive_FloatingDown_29_#7:Gravity Override Setting Struct Structure=Make Gravity Override Setting Struct(
     WindFrequency_28_#2="1.5,7,8"
     WindAmplitude_31_#3="200,-500,400"
     WindOffset_32_#4="0,-600,0"
     GravityOverride_15_#5="0,-300,-600"
     JointName_27_#6=root
    )
    Skydive_DivingDown_30_#8:Gravity Override Setting Struct Structure=Make Gravity Override Setting Struct(
     WindFrequency_28_#2="1.5,12,10"
     WindAmplitude_31_#3="0,50,50"
     WindOffset_32_#4="0,0,0"
     GravityOverride_15_#5="-1500,0,0"
     JointName_27_#6=pelvis
    )
    Skydive_Parachuting_31_#9:Gravity Override Setting Struct Structure=Make Gravity Override Setting Struct(
     WindFrequency_28_#2="1.5,7,8"
     WindAmplitude_31_#3="50,-100,400"
     WindOffset_32_#4="0,0,0"
     GravityOverride_15_#5="0,-300,-980"
     JointName_27_#6=root
    )
   ) -> (then:Exec, StructOut:"Gravity Override Params Struct Structure (by ref)") {is=title K2Node_SetFieldsInStruct}
   > Set Gravity Override Params(Gravity Override Params:Gravity Override Params Struct Structure=@n9) -> (Output_Get:Gravity Override Params Struct Structure)
   n10 = Set members in Gravity Override Params Struct(
    execute=@n8.else
    StructRef=Get Gravity Override Params()
    Skydive_GoingUp_28_#1=Make Gravity Override Setting Struct(
     WindFrequency_28_#2="1.5,7,8"
     WindAmplitude_31_#3="200,-500,400"
     WindOffset_32_#4="0,0,0"
     GravityOverride_15_#5="0,-300,1500"
     JointName_27_#6=root
    )
    Skydive_FloatingDown_29_#7=Make Gravity Override Setting Struct(
     WindFrequency_28_#2="1.5,7,8"
     WindAmplitude_31_#3="200,-500,400"
     WindOffset_32_#4="0,-600,400"
     GravityOverride_15_#5="0,-700,600"
     JointName_27_#6=root
    )
    Skydive_DivingDown_30_#8=Make Gravity Override Setting Struct(
     WindFrequency_28_#2="1.5,12,10"
     WindAmplitude_31_#3="50,300,100"
     WindOffset_32_#4="0,0,0"
     GravityOverride_15_#5="-1500,0,0"
     JointName_27_#6=pelvis
    )
    Skydive_Parachuting_31_#9=Make Gravity Override Setting Struct(
     WindFrequency_28_#2="1.5,7,8"
     WindAmplitude_31_#3="200,-500,400"
     WindOffset_32_#4="0,0,0"
     GravityOverride_15_#5="0,-800,-980"
     JointName_27_#6=root
    )
   )
   > Set Gravity Override Params(Gravity Override Params=@n10)
   n11 = CheckForBackbling()
   n12 = Set IsWearingBackbling(execute=@n7.then_0, IsWearingBackbling:bool=@n11) -> (Output_Get:bool)
   > Set Is Backbling A Cape(Is Backbling A Cape:bool=@n11.bIsBackblingACape) -> (Output_Get:bool)
   > Set NotBackbling(NotBackbling:double="To Float (Boolean)"(InBool=NOT Boolean(A=@n11))) -> (Output_Get:double)
   > Set is Wear a Cape(is Wear a Cape:bool=AND Boolean(A=@n11, B=@n11.bIsBackblingLong)) -> (Output_Get:bool)
   > n13 = Branch(Condition=@n12)
   > n14 = Branch(Condition=@n11.bIsBackblingLong)
   Set IsShortBackblingOnAlpha(execute=[@n14.then, @n13.else], IsShortBackblingOnAlpha:double=0) -> (Output_Get:double)
   Set IsShortBackblingOnAlpha(execute=@n14.else, IsShortBackblingOnAlpha=1)
   n15 = Event OnCharacterCustomizationLoaded()
   Event OnBeginJumping()
   Event OnEndJumping()
   Set Component Linear Acc Scale(Component Linear Acc Scale:Vector="0.4,0.7,0.3") -> (Output_Get:Vector)
   Set Component Linear Acc Scale(Component Linear Acc Scale="0.3,0.25,0.5")
  ]}
]
variables [
 AWS {type=boolean}
 Component Linear Acc Scale {type=Vector, default=<0.3 0.15 0.2>}
 Gravity Override {type=Vector, default=<0.3 0.25 0.5>}
 Gravity Override Params {type=GravityOverrideParamsStruct}
 HeadUpAlpha {type=number}
 Is Backbling A Cape {type=boolean}
 is Wear a Cape {type=boolean}
 IsShortBackblingOnAlpha {type=number}
 IsWearingBackbling {type=boolean}
 NotBackbling {type=number}
 PonyTailCorrAlpha {type=number}
]
